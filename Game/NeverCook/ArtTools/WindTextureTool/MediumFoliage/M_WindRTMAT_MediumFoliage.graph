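# Unlit art-tool material for medium foliage that builds a wind texture, apparently for render targets, packing direction in RG and strength in B.
# Two four-octave sine sets along U are blended by a stepped, Deadzone-adjusted sine mask along V.
# That wave perturbs the MPC_Wind direction sideways by WindNoise.
# Separate noise-texture branches build overlay-blended strength and heightmap-derived normals; no link from any branch to a material output appears.
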Material /Game/NeverCook/ArtTools/WindTextureTool/MediumFoliage/M_WindRTMAT_MediumFoliage
graph {
 properties {pixelDepthOffsetMode=PDOM_Legacy, shadingModel=MSM_Unlit, usedShadingModels=Unlit}
 nodes [
  n1 = Constant3Vector() {constant=<rgba 1 1 0 0>}
  n2 = ComponentMask(None=CollectionParameter() {collection=/Game/Environments/Asteria/Foliage/Wind/MPC_Wind, parameterName=WindDirection}) {b=true, g=true, r=true}
  n3 = Constant3Vector() {constant=<rgba 0 0 1 0>}
  n4 = ComponentMask(None=TextureCoordinate()) {r=true}
  n5 = Sine(None=Multiply(A=@n4) {constB=2})
  n6 = Multiply(A=ScalarParameter() {defaultValue=1, parameterName=Huge2, sliderMax=1}, B=Divide(A=Add(A=@n5)))
  n7 = Add(A=Sine(None=Multiply(A=@n4) {constB=4}))
  n8 = LinearInterpolate(A=@n6, B=Add(A=@n6, B=Divide(A=@n7)), Alpha=ScalarParameter() {defaultValue=0.75, parameterName=Big2})
  n9 = Add(A=Sine(None=Multiply(A=@n4) {constB=8}))
  n10 = LinearInterpolate(A=@n8, B=Add(A=@n8, B=Divide(A=@n9)), Alpha=ScalarParameter() {defaultValue=0.45, parameterName=Medium2})
  n11 = Multiply(A=@n4) {constB=16}
  n12 = Divide(A=Add(A=Sine(None=@n11)))
  n13 = ComponentMask(None=TextureCoordinate()) {r=true}
  n14 = Sine(None=Multiply(A=@n13) {constB=2})
  n15 = Multiply(A=ScalarParameter() {defaultValue=1, parameterName=Huge, sliderMax=1}, B=Divide(A=Add(A=@n14)))
  n16 = Add(A=Sine(None=Multiply(A=@n13) {constB=4}))
  n17 = LinearInterpolate(A=@n15, B=Add(A=@n15, B=Divide(A=@n16)), Alpha=ScalarParameter() {defaultValue=0.75, parameterName=Big})
  n18 = Add(A=Sine(None=Multiply(A=@n13) {constB=8}))
  n19 = LinearInterpolate(A=@n17, B=Add(A=@n17, B=Divide(A=@n18)), Alpha=ScalarParameter() {defaultValue=0.45, parameterName=Medium})
  n20 = Multiply(A=@n13) {constB=16}
  n21 = Divide(A=Add(A=Sine(None=@n20)))
  n22 = Add(A=ComponentMask(None=TextureCoordinate()) {g=true}) {constB=0.25}
  n23 = Add(A=Multiply(A=Sine(None=@n22), B=ScalarParameter() {defaultValue=1.25, parameterName=Deadzone}))
  n24 = Constant() {r=2}
  n25 = Multiply(A=Saturate(None=Divide(A=@n23)), B=@n24) {constB=4}
  n26 = LinearInterpolate(
   A=LinearInterpolate(A=@n10, B=Add(A=@n10, B=@n12), Alpha=ScalarParameter() {defaultValue=0.2, parameterName=Small2})
   B=LinearInterpolate(A=@n19, B=Add(A=@n19, B=@n21), Alpha=ScalarParameter() {defaultValue=0.2, parameterName=Small})
   Alpha=Divide(A=Floor(None=@n25), B=@n24) {constB=4}
  )
  n27 = Subtract(A=ComponentMask(None=@n26) {r=true}) {constB=0.5}
  n28 = ScalarParameter() {defaultValue=0.05, parameterName=WindNoise}
  n29 = Add(A=@n2, B=Multiply(A=CrossProduct(A=@n2, B=@n3), B=Multiply(A=@n27, B=@n28)))
  n30 = Add(A=Normalize(VectorInput=Multiply(A=@n1, B=@n29)))
  AppendVector(A=ComponentMask(None=Divide(A=@n30)) {g=true, r=true}, B=ComponentMask(None=@n26) {r=true})
  n31 = Multiply(
   A=ScalarParameter() {defaultValue=1, parameterName=Huge, sliderMax=1}
   B=TextureSample() {samplerType=SAMPLERTYPE_LinearColor, texture=/Game/NeverCook/ArtTools/WindTextureTool/WindNoise_Mask_5}.R
  )
  n32 = LinearInterpolate(
   A=@n31
   B=MaterialFunctionCall(Base=@n31, Blend=TextureSample() {samplerType=SAMPLERTYPE_LinearColor, texture=/Game/NeverCook/ArtTools/WindTextureTool/WindNoise_Mask_2}.R) {materialFunction=&Blend_Overlay}.Result
   Alpha=ScalarParameter() {defaultValue=0.75, parameterName=Big}
  )
  n33 = LinearInterpolate(
   A=@n32
   B=MaterialFunctionCall(Base=@n32, Blend=TextureSample() {samplerType=SAMPLERTYPE_LinearColor, texture=/Game/NeverCook/ArtTools/WindTextureTool/WindNoise_Mask_3}.R) {materialFunction=&Blend_Overlay}.Result
   Alpha=ScalarParameter() {defaultValue=0.45, parameterName=Medium}
  )
  LinearInterpolate(
   A=@n33
   B=MaterialFunctionCall(Base=@n33, Blend=TextureSample() {samplerType=SAMPLERTYPE_LinearColor, texture=/Game/NeverCook/ArtTools/WindTextureTool/WindNoise_Mask_4}.R) {materialFunction=&Blend_Overlay}.Result
   Alpha=ScalarParameter() {defaultValue=0.2, parameterName=Small}
  )
  n34 = ScalarParameter() {defaultValue=2.7977, parameterName=WindDirectionNoise}
  n35 = ScalarParameter() {defaultValue=0.005}
  n36 = LinearInterpolate(
   A=Constant3Vector() {constant=<rgba 0 0 1 0>}
   B=MaterialFunctionCall!(
    Height Map=TextureObject() {samplerType=SAMPLERTYPE_Masks, texture=/Game/Global/Textures/Default/Noise/WindNoise_Mask_1}
    Normal Map Intensity=@n34
    Height Map UV Offset=@n35
    Coordinates
    Height Map Channel Selector
   ) {materialFunction=/Engine/Functions/Engine_MaterialFunctions03/Procedurals/NormalFromHeightmap}.Result
   Alpha=ScalarParameter() {defaultValue=1, parameterName=Huge, sliderMax=1}
  )
  n37 = TextureObject() {samplerType=SAMPLERTYPE_Masks, texture=&WindNoise_Mask_2}
  n38 = LinearInterpolate(
   A=@n36
   B=MaterialFunctionCall(
    BaseNormal=@n36
    AdditionalNormal=MaterialFunctionCall(Height Map=@n37, Normal Map Intensity=@n34, Height Map UV Offset=@n35) {materialFunction=&NormalFromHeightmap}.Result
   ) {materialFunction=&BlendAngleCorrectedNormals}.Result
   Alpha=ScalarParameter() {defaultValue=0.75, parameterName=Big}
  )
  n39 = TextureObject() {samplerType=SAMPLERTYPE_Masks, texture=&WindNoise_Mask_3}
  n40 = LinearInterpolate(
   A=@n38
   B=MaterialFunctionCall(
    BaseNormal=@n38
    AdditionalNormal=MaterialFunctionCall(Height Map=@n39, Normal Map Intensity=@n34, Height Map UV Offset=@n35) {materialFunction=&NormalFromHeightmap}.Result
   ) {materialFunction=&BlendAngleCorrectedNormals}.Result
   Alpha=ScalarParameter() {defaultValue=0.45, parameterName=Medium}
  )
  n41 = MaterialFunctionCall(
   BaseNormal=@n40
   AdditionalNormal=MaterialFunctionCall(
    Height Map=TextureObject() {samplerType=SAMPLERTYPE_Masks, texture=&WindNoise_Mask_4}
    Normal Map Intensity=@n34
    Height Map UV Offset=@n35
   ) {materialFunction=&NormalFromHeightmap}.Result
  ) {materialFunction=&BlendAngleCorrectedNormals}
  n42 = ScalarParameter() {defaultValue=0.2, parameterName=Small}
  n43 = ComponentMask(None=Normalize(VectorInput=LinearInterpolate(A=@n40, B=@n41.Result, Alpha=@n42))) {g=true, r=true}
  n44 = Constant() {r=1}
  Multiply(A=Normalize(VectorInput=AppendVector(A=@n43, B=@n44)), B=@n26)
 ]
}
